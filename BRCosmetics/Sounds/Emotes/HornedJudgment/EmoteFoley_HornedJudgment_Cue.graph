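# Foley sound cue for the Horned Judgment emote, pairing a start sound with a looping traversal layer.
# A mixer plays the one-shot start alongside the loop, which a parameter crossfade fades in as EmoteMoveSpeed rises from zero.
# The loop branch overrides its sound class to emote music, while the cue uses emote foley.
# The cue has emote attenuation and concurrency limits, and sends to a music audio-analysis submix.
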SoundCue /BRCosmetics/Sounds/Emotes/HornedJudgment/EmoteFoley_HornedJudgment_Cue
root=Mixer_0
nodes [
 {id=Mixer_0, children [WavePlayer_1,SoundClass_0], properties {inputVolume [0.9,0.9]}}
 {
  id=ParamCrossFade_0
  children [WavePlayer_3]
  properties {
   paramName=EmoteMoveSpeed
   crossFadeInput [{fadeInDistanceEnd=1, fadeInDistanceStart=0, fadeOutDistanceEnd=100, fadeOutDistanceStart=100, volume=1}]
  }
 }
 {id=SoundClass_0, children [ParamCrossFade_0], properties {soundClassOverride=&Fort_Music_Emote}}
 {id=WavePlayer_1, properties {soundWaveAssetPtr=/Game/Athena/Sounds/Emotes/HornedJudgment/Emote_HornedJudgment_Start}}
 {id=WavePlayer_3, properties {bLooping=true, soundWaveAssetPtr=&/Emote_HornedJudgment_Traversal_Loop}}
]
properties {
 attenuationSettings=&Fort_Emotes_Athena_Att
 audioPropertiesSheet=/BRCosmetics/Sounds/Settings/PropertySheets/Emotes/APS_Cosmetic_Emote_Foley
 maxDistance=1600
 soundClassObject=/Game/Sounds/FortSoundClassesAndModes/Fort_Foley_Emote
 volumeMultiplier=1
 concurrencyOverrides {maxCount=3}
 concurrencySet [/Game/Sounds/Concurrency_Rules/Athena_EmoteFoley_Conc]
 soundSubmixSends [{minSendLevel=0, soundSubmix=&Fortnite_AudioAnalysis_Music}]
}
